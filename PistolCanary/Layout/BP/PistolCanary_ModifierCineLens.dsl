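; PistolCanary camera modifier that applies a cine-lens look, reshaping the view and depth of field as it blends in.
; BlueprintModifyCamera lerps the view location toward itself transformed by Camera_Up and FOV toward the configured value, leaving rotation unchanged.
; BlueprintModifyPostProcess supplies post-process settings built from the aperture (F-stop) and focal distance variables.
; The modifier's Alpha drives both camera lerps and the post-process blend weight.
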(fn BlueprintModifyCamera
  (params (DeltaTime :type Float%20%28single-precision%29) (ViewLocation :type Vector :default "0, 0, 0") (ViewRotation :type Rotator :default "0, 0, 0") (FOV :type Float%20%28single-precision%29))
  (return (Math|Vector|Lerp(Vector) ViewLocation (Math|Transform|TransformLocation (Variables|Default|GetCamera_Up) ViewLocation) (Variables|CameraModifier|GetAlpha)) ViewRotation (Math|Float|Lerp FOV (Variables|Default|GetFOV) (Variables|CameraModifier|GetAlpha))))

(fn BlueprintModifyPostProcess
  (params (DeltaTime :type Float%20%28single-precision%29))
  (return (Math|Float|Lerp 0.0 1.0 (Variables|CameraModifier|GetAlpha)) (Utilities|Struct|MakePostProcessSettings (Variables|Default|GetAperture%20%28F-stop%29) (Variables|Default|GetFocal%20Distance))))
