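; Camera modifier for PistolCanary that, judging by its name, handles a whip transition to the game camera.
; It lerps the view location toward a point offset along player character 0's right vector by the Distance variable, using the modifier's Alpha as the blend, leaving rotation and FOV unchanged.
; It also returns fixed post-process settings at full blend weight, independent of Alpha.
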(fn BlueprintModifyCamera
  (params (DeltaTime :type Float%20%28single-precision%29) (ViewLocation :type Vector :default "0, 0, 0") (ViewRotation :type Rotator :default "0, 0, 0") (FOV :type Float%20%28single-precision%29))
  (return (Math|Vector|Lerp(Vector) ViewLocation (+ ViewLocation (* (Transformation|GetActorRightVector (Game|GetPlayerCharacter 0)) (Variables|Default|GetDistance))) (Variables|CameraModifier|GetAlpha)) ViewRotation FOV))

(fn BlueprintModifyPostProcess
  (params (DeltaTime :type Float%20%28single-precision%29))
  (return 1.0 (Utilities|Struct|MakePostProcessSettings 0.15 1e+02)))
